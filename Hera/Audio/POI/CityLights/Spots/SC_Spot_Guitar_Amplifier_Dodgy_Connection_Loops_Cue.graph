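# Ambient spot SoundCue of a guitar amplifier with a dodgy connection, for the CityLights POI.
# It mixes one randomly chosen looping bed with randomly picked interference one-shots that repeat after random delays.
# The mix passes through a volume modulator and endlessly looping envelope for subtle level variation, feeding only the first two quality slots; the others are empty.
# Overrides add logarithmic attenuation with occlusion and distance filtering, plus limited concurrency.
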SoundCue /Hera/Audio/POI/CityLights/Spots/SC_Spot_Guitar_Amplifier_Dodgy_Connection_Loops_Cue
root=QualityLevel_1
nodes [
 {id=Delay_0, children [Random_4], properties {delayMax=10, delayMin=5}}
 {
  id=Enveloper_0
  children [Modulator_0]
  properties {
   bLoop=true
   bLoopIndefinitely=true
   loopEnd=20
   pitchCurve {editorCurveData {keys [{time=0}]}}
   volumeCurve {editorCurveData {keys [<at -0.0028 1>,<at 0 1>,<at 10 0.95>,<at 20 1>]}}
  }
 }
 {id=Looping_0, children [Delay_0]}
 {id=Mixer_0, children [Looping_0,Random_0], properties {inputVolume [0.5,0.8]}}
 {id=Modulator_0, children [Mixer_0], properties {volumeMax=1, volumeMin=0.9}}
 {id=QualityLevel_1, children [Enveloper_0,Enveloper_0,null,null]}
 {id=Random_0, children [WavePlayer_0,WavePlayer_1,WavePlayer_3], properties {weights [1,1,1]}}
 {
  id=Random_4
  children [WavePlayer_27, WavePlayer_28, WavePlayer_29, WavePlayer_30, WavePlayer_31, WavePlayer_32, WavePlayer_34, WavePlayer_35, WavePlayer_36]
  properties {weights [1,1,1,1,1,1,1,1,1]}
 }
 {id=WavePlayer_0, properties {bLooping=true, soundWaveAssetPtr=&/Spot_Guitar_Amp_Dodgy_Connection_Loop_01}}
 {id=WavePlayer_1, properties {bLooping=true, soundWaveAssetPtr=&/Spot_Guitar_Amp_Dodgy_Connection_Loop_02}}
 {
  id=WavePlayer_27
  properties {soundWaveAssetPtr=/AmbientAudioContent/SoundCues/OneShots/Spot_Guitar_Amplifier_Idling_Interference_Oneshot_10}
 }
 {
  id=WavePlayer_28
  properties {soundWaveAssetPtr=/AmbientAudioContent/SoundCues/OneShots/Spot_Guitar_Amplifier_Idling_Interference_Oneshot_09}
 }
 {
  id=WavePlayer_29
  properties {soundWaveAssetPtr=/AmbientAudioContent/SoundCues/OneShots/Spot_Guitar_Amplifier_Idling_Interference_Oneshot_08}
 }
 {id=WavePlayer_3, properties {bLooping=true, soundWaveAssetPtr=&/Spot_Guitar_Amp_Dodgy_Connection_Loop_04}}
 {
  id=WavePlayer_30
  properties {soundWaveAssetPtr=/AmbientAudioContent/SoundCues/OneShots/Spot_Guitar_Amplifier_Idling_Interference_Oneshot_07}
 }
 {
  id=WavePlayer_31
  properties {soundWaveAssetPtr=/AmbientAudioContent/SoundCues/OneShots/Spot_Guitar_Amplifier_Idling_Interference_Oneshot_06}
 }
 {
  id=WavePlayer_32
  properties {soundWaveAssetPtr=/AmbientAudioContent/SoundCues/OneShots/Spot_Guitar_Amplifier_Idling_Interference_Oneshot_05}
 }
 {
  id=WavePlayer_34
  properties {soundWaveAssetPtr=/AmbientAudioContent/SoundCues/OneShots/Spot_Guitar_Amplifier_Idling_Interference_Oneshot_03}
 }
 {
  id=WavePlayer_35
  properties {soundWaveAssetPtr=/AmbientAudioContent/SoundCues/OneShots/Spot_Guitar_Amplifier_Idling_Interference_Oneshot_02}
 }
 {
  id=WavePlayer_36
  properties {soundWaveAssetPtr=/AmbientAudioContent/SoundCues/OneShots/Spot_Guitar_Amplifier_Idling_Interference_Oneshot_01}
 }
]
properties {
 bDebug=true
 bEnableSubmixSends=false
 bOverrideAttenuation=true
 bOverrideConcurrency=true
 maxDistance=650
 soundClassObject=&Fort_Ambient_Props_Athena
 volumeMultiplier=1
 attenuationOverrides {
  bAttenuateWithLPF=true
  bEnableLogFrequencyScaling=true
  bEnableOcclusion=true
  distanceAlgorithm=Logarithmic
  falloffDistance=500
  hPFFrequencyAtMax=300
  lPFRadiusMax=1000
  lPFRadiusMin=250
  occlusionLowPassFilterFrequency=850
  occlusionTraceChannel=ECC_GameTraceChannel11
  attenuationShapeExtents {x=150}
 }
 concurrencyOverrides {maxCount=2}
 concurrencySet [/AmbientAudioBase/Concurrency/AmbientConc_1Max_Prop_StopLowestPriority_Concurrency]
}
