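# Sound cue for a looping idle guitar-amp static spot ambience in the CityLights POI, classed as an Athena ambient prop.
# A looping wave passes through slight random volume modulation and an endlessly looping envelope that gently dips and recovers.
# A quality switch feeds that chain to the first two levels only, leaving the other two empty.
# Overrides set capsule attenuation with distance filtering, listener focus and a low concurrency cap.
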SoundCue /Hera/Audio/POI/CityLights/Spots/SC_Spot_Guitar_Idle_Static_Loops_Cue
root=QualityLevel_1
nodes [
 {
  id=Enveloper_0
  children [Modulator_0]
  properties {
   bLoop=true
   bLoopIndefinitely=true
   loopEnd=20
   pitchCurve {editorCurveData {keys [{time=0}]}}
   volumeCurve {editorCurveData {keys [{arriveTangent=-0.005, interpMode=RCIM_Cubic, time=0, value=1}, <at 10 0.95 RCIM_Cubic>, <at 20 1>]}}
  }
 }
 {id=Modulator_0, children [WavePlayer_2], properties {volumeMax=1, volumeMin=0.9}}
 {id=QualityLevel_1, children [Enveloper_0,Enveloper_0,null,null]}
 {id=WavePlayer_2, properties {bLooping=true, soundWaveAssetPtr=&/Spot_Guitar_Amp_Static_Loop_01}}
]
properties {
 bDebug=true
 bEnableSubmixSends=false
 bOverrideAttenuation=true
 bOverrideConcurrency=true
 maxDistance=604.2747
 soundClassObject=&Fort_Ambient_Props_Athena
 volumeMultiplier=1
 attenuationOverrides {
  attenuationShape=Capsule
  bAttenuateWithLPF=true
  bEnableListenerFocus=true
  bEnableLogFrequencyScaling=true
  falloffDistance=500
  focusVolumeAttenuation=1.2
  hPFFrequencyAtMax=300
  hPFFrequencyAtMin=50
  lPFRadiusMax=500
  lPFRadiusMin=250
  nonFocusVolumeAttenuation=0.4
  occlusionLowPassFilterFrequency=850
  occlusionTraceChannel=ECC_GameTraceChannel11
  stereoSpread=0
  attenuationShapeExtents=<104.2747 50>
 }
 concurrencyOverrides {maxCount=2}
 concurrencySet [/AmbientAudioBase/Concurrency/AmbientConc_1Max_Prop_StopLowestPriority_Concurrency]
}
