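# Sound cue for a positional guitar-sign ambience spot in the CityLights POI, built from amp hum, crackle and feedback.
# A delayed, looping feedback-bend layer mixes with a random pick of either looping dodgy-connection variants or one of four hum loops.
# Looping volume envelopes gently swell the picked bed, and a modulator varies overall level.
# Only the first two quality levels play it, with overridden attenuation, occlusion and concurrency.
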SoundCue /Hera/Audio/POI/CityLights/Spots/SC_Spot_Guitar_Sign_Feedback_Bending_Loop_Cue
root=QualityLevel_1
nodes [
 {id=Delay_0, children [WavePlayer_5], properties {delayMax=18, delayMin=12}}
 {id=Delay_1, children [Random_2], properties {delayMax=1, delayMin=0.5}}
 {
  id=Enveloper_1
  children [Looping_6]
  properties {
   bLoop=true
   bLoopIndefinitely=true
   loopEnd=20
   pitchCurve {editorCurveData {keys [{time=0}]}}
   volumeCurve {editorCurveData {keys [{arriveTangent=-0.005, interpMode=RCIM_Cubic, time=0, value=1}, <at 10 0.95 RCIM_Cubic>, <at 20 1>]}}
  }
 }
 {
  id=Enveloper_2
  children [Random_5]
  properties {
   bLoop=true
   bLoopIndefinitely=true
   loopEnd=20
   pitchCurve {editorCurveData {keys [{time=0}]}}
   volumeCurve {editorCurveData {keys [{arriveTangent=-0.005, interpMode=RCIM_Cubic, time=0, value=1}, <at 10 0.95 RCIM_Cubic>, <at 20 1>]}}
  }
 }
 {id=Looping_6, children [Delay_1]}
 {id=Looping_7, children [Delay_0]}
 {id=Mixer_0, children [Looping_7,Random_3], properties {inputVolume [1,1]}}
 {id=Mixer_1, children [Enveloper_1], properties {inputVolume [0.75]}}
 {id=Modulator_0, children [Mixer_0], properties {volumeMax=1, volumeMin=0.9}}
 {id=Modulator_2, children [WavePlayer_23], properties {volumeMax=0.5, volumeMin=0.5}}
 {id=Modulator_3, children [WavePlayer_24], properties {volumeMax=1.2, volumeMin=1.2}}
 {id=QualityLevel_1, children [Modulator_0,Modulator_0,null,null]}
 {id=Random_2, children [WavePlayer_6,WavePlayer_7], properties {weights [1,1]}}
 {id=Random_3, children [Mixer_1,Enveloper_2], properties {weights [1,1]}}
 {id=Random_5, children [WavePlayer_22, Modulator_2, Modulator_3, WavePlayer_25], properties {weights [1,1,1,1]}}
 {id=WavePlayer_22, properties {bLooping=true, soundWaveAssetPtr=/AmbientAudioContent/SoundCues/SpottedLoops/Spot_Guitar_Amp_Hum_Loop_01}}
 {id=WavePlayer_23, properties {bLooping=true, soundWaveAssetPtr=/AmbientAudioContent/SoundCues/SpottedLoops/Spot_Guitar_Amp_Hum_Loop_02}}
 {id=WavePlayer_24, properties {bLooping=true, soundWaveAssetPtr=/AmbientAudioContent/SoundCues/SpottedLoops/Spot_Guitar_Amp_Hum_Loop_03}}
 {id=WavePlayer_25, properties {bLooping=true, soundWaveAssetPtr=/AmbientAudioContent/SoundCues/SpottedLoops/Spot_Guitar_Amp_Hum_Loop_04}}
 {id=WavePlayer_5, properties {soundWaveAssetPtr=&/Spot_Guitar_Amp_Feedback_Bending_Rarely_Loop}}
 {id=WavePlayer_6, properties {soundWaveAssetPtr=&/Spot_Guitar_Amp_Dodgy_Connection_Loop_01}}
 {id=WavePlayer_7, properties {soundWaveAssetPtr=&/Spot_Guitar_Amp_Dodgy_Connection_Loop_02}}
]
properties {
 bDebug=true
 bEnableSubmixSends=false
 bOverrideAttenuation=true
 bOverrideConcurrency=true
 maxDistance=850
 soundClassObject=/Game/Sounds/FortSoundClassesAndModes/Fort_Ambient_OneShots_Indoors
 volumeMultiplier=1
 attenuationOverrides {
  bAttenuateWithLPF=true
  bEnableListenerFocus=true
  bEnableLogFrequencyScaling=true
  bEnableOcclusion=true
  distanceAlgorithm=Logarithmic
  falloffDistance=700
  hPFFrequencyAtMax=300
  lPFRadiusMax=1000
  lPFRadiusMin=250
  nonFocusVolumeAttenuation=0.4
  occlusionLowPassFilterFrequency=850
  occlusionTraceChannel=ECC_GameTraceChannel11
  stereoSpread=50
  attenuationShapeExtents=<150 90>
 }
 concurrencyOverrides {maxCount=2}
 concurrencySet [/AmbientAudioBase/Concurrency/AmbientConc_1Max_Prop_StopLowestPriority_Concurrency]
}
