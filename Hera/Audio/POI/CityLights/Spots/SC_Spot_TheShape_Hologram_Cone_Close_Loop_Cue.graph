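# Looping spot ambience cue for the close cone layer of TheShape hologram in the CityLights POI, routed to the outdoor ambient sound class.
# One looping hologram wave, pitched up by a fixed modulator, feeds two parallel attenuations: a small cone and a small occluded sphere.
# A mixer weights the cone branch above the sphere, then an enveloper ramps volume in.
# A concurrency override limits the cue to a single instance.
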SoundCue /Hera/Audio/POI/CityLights/Spots/SC_Spot_TheShape_Hologram_Cone_Close_Loop_Cue
root=Enveloper_0
nodes [
 {id=Attenuation_0, children [Modulator_2], properties {attenuationSettings=&/TheShape_Hologram_Small_Cone_Att}}
 {
  id=Attenuation_2
  children [Modulator_2]
  properties {attenuationSettings=&/TheShape_Hologram_Small_Sphere_Occ_Att}
 }
 {
  id=Enveloper_0
  children [Mixer_0]
  properties {pitchCurve {editorCurveData {keys [{time=0}]}}, volumeCurve {editorCurveData {keys [<at 0 0>,<at 1 1>]}}}
 }
 {id=Mixer_0, children [Attenuation_0, Attenuation_2], properties {inputVolume [2,1]}}
 {id=Modulator_2, children [WavePlayer_2], properties {pitchMax=3, pitchMin=3, volumeMax=1, volumeMin=1}}
 {id=WavePlayer_2, properties {bLooping=true, soundWaveAssetPtr=/AmbientAudioContent/Waves/POI/Spot_NeonCity_Hologram_UltraSam_Loop}}
]
properties {
 bOverrideConcurrency=true
 maxDistance=3000
 soundClassObject=&Fort_Ambient_Outdoors_Athena
 volumeMultiplier=0.5
 attenuationOverrides {falloffDistance=3998.0466}
 concurrencyOverrides {maxCount=1}
 concurrencySet [&Weapon_Drop_StopOldest]
}
