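# Looping ambient spot cue for TheShape hologram at the CityLights POI, classed as outdoor Athena ambience.
# A looping giant hologram hum-buzz wave feeds a Modulator that pins pitch at triple with no random variation.
# The cue plays at half volume through a medium hologram attenuation asset.
# A concurrency override caps it at one instance, superseding the assigned Weapon_Drop_StopOldest set.
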SoundCue /Hera/Audio/POI/CityLights/Spots/SC_Spot_TheShape_Hologram_Loop_Cue
root=Modulator_2
nodes [
 {id=Modulator_2, children [WavePlayer_2], properties {pitchMax=3, pitchMin=3, volumeMax=1, volumeMin=1}}
 {id=WavePlayer_2, properties {bLooping=true, soundWaveAssetPtr=/AmbientAudioContent/Waves/SpottedLoops/Spot_Hologram_Giant_Hum_Buzz_Loop_01}}
]
properties {
 attenuationSettings=&/TheShape_Hologram_Medium_Occ_Att
 bOverrideConcurrency=true
 maxDistance=4000
 soundClassObject=&Fort_Ambient_Outdoors_Athena
 volumeMultiplier=0.5
 attenuationOverrides {falloffDistance=3998.0466}
 concurrencyOverrides {maxCount=1}
 concurrencySet [&Weapon_Drop_StopOldest]
}
